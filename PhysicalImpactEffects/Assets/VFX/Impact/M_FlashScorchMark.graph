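# Deferred decal material for a scorch mark impact effect, with parameters for color, metallic, specular and roughness.
# A weighted sum of ground-char and scorch-hole texture red channels forms a saturated mask, multiplied by decal lifetime opacity.
# An optional projection fade compares world normal to the decal axis to hide stretching.
# A shading path switch multiplies the mask by a rerouted value only on the default path, leaving forward and mobile unmodified.
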Material /PhysicalImpactEffects/Assets/VFX/Impact/M_FlashScorchMark
graph {
 properties {bAutomaticallySetUsageInEditor=false, blendMode=BLEND_Translucent, materialDomain=MD_DeferredDecal, usedShadingModels=Default Lit}
 nodes [
  n1 = NamedRerouteUsage()
  n2 = TextureSample() {samplerType=SAMPLERTYPE_LinearColor, texture=/Game/Effects/Fort_Effects/Gameplay/SmokeDustExplosions/Textures/T_GroundChar_01}
  n3 = ScalarParameter() {defaultValue=0.5}
  n4 = TextureSample() {texture=/Game/Effects/Fort_Effects/Gameplay/SmokeDustExplosions/Textures/T_Decal_Scorch_Hole}
  n5 = ScalarParameter() {defaultValue=0.5, parameterName=Param_1}
  n6 = Saturate(None=Add(A=Multiply(A=@n2.R, B=@n3) {constB=5}, B=Multiply(A=@n4.R, B=@n5)))
  Multiply(A=ShadingPathSwitch(Default=Multiply(A=@n1, B=@n6), Forward=@n6, Mobile=@n6), B=DecalLifetimeOpacity().Opacity)
  n7 = ScalarParameter() {
   desc="This parameter drives some fade effects based on the anngle of or decal projection axis and thre wWorld Normal to avoid seeing harsh texture stretch, Use 1 to Enable, 0 To Disable"
   group=Opacity
   parameterName=Decal Projection Fade Enabled
  }
  n8 = ComponentMask(None=SceneTexture() {sceneTextureId=PPI_WorldNormal}.Color) {b=true, g=true, r=true}
  n9 = Normalize(VectorInput=Transform(None=Constant3Vector() {constant=<rgba -1 0 0 0>}) {transformSourceType=TRANSFORMSOURCE_Local})
  n10 = ScalarParameter() {
   defaultValue=1
   desc=Projection Fade power value applied by the dot product with the world normal and the decal projection axis.
   group=Opacity
   parameterName=Decal Projection Fade Power
  }
  n11 = Power(Base=Saturate(None=DotProduct(A=@n8, B=@n9)), Exp=@n10) {constExponent=0}
  NamedRerouteDeclaration(None=If(A=@n7, "A > B"=@n11, "A == B"=@n11, "A < B"=Constant() {r=1}) {constB=0.99}) {name=Decal Projection Fade, nodeColor=<rgba 0.1953 1 0 1>}
  ScalarParameter() {parameterName=Metallic}
  ScalarParameter() {parameterName=Specular}
  ScalarParameter() {defaultValue=1, parameterName=Roughness}
  n12 = TextureCoordinate()
  SphereMask(A=ShadingPathSwitch(Default=LightVector(), Forward=@n12, Mobile=@n12), B=Constant() {r=0.5}, Radius=Constant() {r=0.5}) {hardnessPercent=90}
  VectorParameter() {parameterName=DecalColor, defaultValue=<rgba 0 0 0 1>}
 ]
}
