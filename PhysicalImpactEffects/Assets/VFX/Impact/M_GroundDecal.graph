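# Translucent deferred decal material for ground cracks in physical impact effects, with tintable color, opacity and normal.
# Opacity combines the packed ARH texture's red channel, a SmoothStep mask on its blue height channel, tint alpha and decal lifetime fade.
# The height thresholds are parameters that default to reversed bounds, Low above High.
# The diffuse texture is multiplied by the tint, and normal-map XY is scaled by NormalStrength, then renormalized.
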Material /PhysicalImpactEffects/Assets/VFX/Impact/M_GroundDecal
graph {
 properties {bAutomaticallySetUsageInEditor=false, blendMode=BLEND_Translucent, materialDomain=MD_DeferredDecal, usedShadingModels=Default Lit}
 nodes [
  Constant()
  Constant() {r=1}
  n1 = VectorParameter() {parameterName=BaseColorAlpha_Tint, defaultValue=<rgba 1 1 1 1>}
  n2 = TextureSampleParameter2D() {parameterName=Tex_ARH, texture=/Game/Effects/Fort_Effects/Textures/GroundCrack/T_Decal_GroundCrack_AlphaRH}
  n3 = SmoothStep(Min=ScalarParameter() {defaultValue=1, parameterName=HeightMask_Low}, Max=ScalarParameter() {parameterName=HeightMask_High}, Value=@n2.B)
  Multiply(A=Multiply(A=@n1.A, B=Multiply(A=@n2.R, B=@n3)), B=DecalLifetimeOpacity().Opacity)
  Multiply(
   A=@n1.RGB
   B=TextureSampleParameter2D() {parameterName=Tex_D, texture=/Game/Effects/Fort_Effects/Textures/GroundCrack/T_Decal_GroundCrack_D}.RGB
  )
  n4 = TextureSampleParameter2D() {parameterName=Tex_N, samplerType=SAMPLERTYPE_Normal, texture=/Game/Effects/Fort_Effects/Textures/GroundCrack/T_Decal_GroundCrack_N}
  n5 = ComponentMask(None=@n4.RGB) {g=true, r=true}
  n6 = ScalarParameter() {defaultValue=1, parameterName=NormalStrength}
  Normalize(VectorInput=AppendVector(A=Multiply(A=@n5, B=@n6), B=@n4.B))
  ParticleColor()
 ]
}
