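# Translucent, unlit, two-sided Niagara mesh-particle material for wind-like impact VFX that tints and distorts the scene behind it.
# Emissive blends scene color with particle color, using particle color alone on forward and mobile paths.
# Alpha masks combine Fresnel, panned noise, a squared depth-fade band and particle alpha through named reroutes.
# Dynamic parameters drive UV panning and scale a dithered, view-space tangent offset for refraction.
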Material /PhysicalImpactEffects/Assets/VFX/Impact/M_SceneDepthWind
graph {
 properties {
  bAutomaticallySetUsageInEditor=false
  bUsedWithNiagaraMeshParticles=true
  blendMode=BLEND_Translucent
  shadingModel=MSM_Unlit
  twoSided=true
  usedShadingModels=Unlit
 }
 nodes [
  n1 = Multiply(A=SceneColor!(UVs).RGB) {constB=2}
  n2 = ParticleColor()
  n3 = Multiply(A=@n2.RGB) {constB=0.75}
  MaterialFunctionCall(Emissive=ShadingPathSwitch(Default=LinearInterpolate(A=@n1, B=@n2.RGB), Forward=@n3, Mobile=@n3)) {materialFunction=&MF_FX_EmissiveAdjustments}
  n4 = Transform(None=VertexTangentWS()) {transformSourceType=TRANSFORMSOURCE_World, transformType=TRANSFORM_View}
  n5 = NamedRerouteUsage()
  Multiply(
   A=Multiply(A=Multiply(A=@n4, B=@n5), B=DynamicParameter() {defaultValue=<rgba 1 1 1 0.1>, paramNames [Time,Distort,Dissolve,Ref]}.Ref)
   B=MaterialFunctionCall() {materialFunction=&DitherTemporalAA}.Result
  )
  n6 = TextureCoordinate()
  n7 = MaterialFunctionCall(XY=@n6) {materialFunction=&Swizzle}
  n8 = ScalarParameter() {defaultValue=1, parameterName=UV2_Swizzle}
  n9 = VectorParameter() {parameterName=UV2, defaultValue=<rgba 0.01 0.36 0.25 0.04>}
  NamedRerouteDeclaration(None=MaterialFunctionCall!(
    ClampTexture
    UV=LinearInterpolate(A=@n6, B=@n7.YX, Alpha=@n8)
    Tex
    InRGB=@n9.RGB
    InA=@n9.RGB
    Time=DynamicParameter() {defaultValue=<rgba 1 1 1 0.1>, paramNames [Time,Distort,Dissolve,Ref]}.Time
    TimeMult
   ) {materialFunction=/Game/Effects/Niagara/MasterMaterials/Functions/MF_FX_VectorDrive_TexPanner}.PannerUV) {name=UV_2, nodeColor=<rgba 0 0.4049 1 1>}
  n10 = TextureCoordinate()
  n11 = MaterialFunctionCall(XY=@n10) {materialFunction=&Swizzle}
  n12 = ScalarParameter() {defaultValue=1, parameterName=Main_UV_Swizzle}
  n13 = VectorParameter() {parameterName=Main_UV, defaultValue=<rgba -0.1612 0.9832 -0.18 0>}
  NamedRerouteDeclaration(None=MaterialFunctionCall(
    UV=LinearInterpolate(A=@n10, B=@n11.YX, Alpha=@n12)
    InRGB=@n13.RGB
    InA=@n13.RGB
    Time=DynamicParameter() {defaultValue=<rgba 1 1 1 0.1>, paramNames [Time,Distort,Dissolve,Ref]}.Time
   ) {materialFunction=&MF_FX_VectorDrive_TexPanner}.PannerUV) {name=MainUV, nodeColor=<rgba 0 0.4049 1 1>}
  n14 = CameraPositionWS()
  n15 = WorldPosition()
  n16 = DotProduct(A=ComponentMask(None=VertexNormalWS()) {b=true, g=true, r=true}, B=Normalize(VectorInput=Subtract(A=@n14, B=@n15.XYZ)))
  n17 = Abs(None=@n16)
  n18 = Power(
   Base=LinearInterpolate(A=OneMinus(None=@n17), B=@n17, Alpha=ScalarParameter() {defaultValue=1, parameterName="Fresnel_Invert (0or1)"})
   Exp=ScalarParameter() {defaultValue=2, parameterName=Fresnel_Power}
  )
  n19 = ScalarParameter() {defaultValue=0.05, parameterName=Fresnel_Offset}
  NamedRerouteDeclaration(None=Multiply(A=Subtract(A=@n18, B=@n19), B=Add(A=@n19))) {name=Fresnel, nodeColor=<rgba 1 0 0.9613 1>}
  n20 = NamedRerouteUsage()
  n21 = TextureSampleParameter2D!(UVs=NamedRerouteUsage(), Bias=ScalarParameter() {defaultValue=5, parameterName=TextureMipBias}, Apply View MipBias) {
   mipValueMode=TMVM_MipBias
   parameterName=NoiseTexture
   samplerType=SAMPLERTYPE_LinearColor
   texture=/Game/Effects/Fort_Effects/Textures/Tile/T_Tile_Directional_T_LightArc
  }
  n22 = LinearInterpolate(A=@n20, B=Multiply(A=@n20, B=@n21.G), Alpha=ScalarParameter() {defaultValue=0.8, parameterName=TextureBlend})
  n23 = NamedRerouteUsage()
  NamedRerouteDeclaration(None=Multiply(A=Multiply(A=@n22, B=@n23), B=NamedRerouteUsage())) {name=Alpha, nodeColor=<rgba 1 0.1875 0 1>}
  n24 = TextureCoordinate()
  n25 = MaterialFunctionCall(XY=@n24) {materialFunction=&Swizzle}
  n26 = ScalarParameter() {defaultValue=1, parameterName=UV1_Swizzle}
  n27 = VectorParameter() {parameterName=UV1, defaultValue=<rgba 0.1033 0.8022 0.332 0>}
  NamedRerouteDeclaration(None=MaterialFunctionCall(
    UV=LinearInterpolate(A=@n24, B=@n25.YX, Alpha=@n26)
    InRGB=@n27.RGB
    InA=@n27.RGB
    Time=DynamicParameter() {defaultValue=<rgba 1 1 1 0.1>, paramNames [Time,Distort,Dissolve,Ref]}.Time
   ) {materialFunction=&MF_FX_VectorDrive_TexPanner}.PannerUV) {name=UV_1, nodeColor=<rgba 0 0.4049 1 1>}
  n28 = Multiply(A=DepthFade(FadeDistance=ScalarParameter() {defaultValue=200, group=Opacity, parameterName=BasedFade_Distance})) {constB=2}
  n29 = OneMinus(None=Abs(None=Subtract(A=@n28)))
  NamedRerouteDeclaration(None=Multiply(A=@n29, B=@n29)) {name=DepthFadeValue, nodeColor=<rgba 0 1 0.5312 1>}
  NamedRerouteDeclaration(None=Multiply(A=ParticleColor().A, B=NamedRerouteUsage())) {name=Opacity, nodeColor=<rgba 0 0 0 1>}
  n30 = TextureCoordinate()
  n31 = TextureCoordinate() {coordinateIndex=1}
  n32 = ScalarParameter() {parameterName=CubeMap_Switch}
  n33 = Subtract(A=Multiply(A=LinearInterpolate(A=@n30, B=@n31, Alpha=@n32)) {constB=2})
  n34 = Multiply(A=@n33, B=@n33) {constB=2}
  n35 = MaterialFunctionCall(Float2=OneMinus(None=Abs(None=@n34))) {materialFunction=&BreakOutFloat2Components}
  NamedRerouteDeclaration(None=MaterialFunctionCall(A=@n35.R, B=Multiply(A=@n35.R, B=@n35.R), C=@n35.R, Alpha=ScalarParameter() {defaultValue=1, parameterName=CubeMap_Select}) {materialFunction=&Lerp_3Color}.Result) {name=CubeMap_2, nodeColor=<rgba 1 1 1 1>}
  NamedRerouteUsage()
  Power(Base=OneMinus(None=TextureCoordinate()), Exp=ScalarParameter() {defaultValue=5, parameterName=AlphaRampPower}) {constExponent=4}
  Reroute(None=@n16)
  VectorParameter() {parameterName=EmissiveOverAllTint, defaultValue=<rgba 1 1 1 1>}
 ]
}
