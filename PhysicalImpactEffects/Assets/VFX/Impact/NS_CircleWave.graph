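# Niagara impact effect that throws a circular shockwave of blur sprites, flipbook smoke, GPU sparks and mesh debris outward from a ring.
# System.RadiusScale scales several burst counts and sprite sizes, while User.DebrisSpawnRate also multiplies both debris counts.
# Debris pops upward and falls under randomized gravity, with age-gated collision only on large chunks.
# FakeVolume adds short-lived expanding sprites delayed by System.BurstTime, and sparks use curl noise, speed-based stretching and camera fade.
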NiagaraSystem /PhysicalImpactEffects/Assets/VFX/Impact/NS_CircleWave
dependencies {
 usedDataInterfaces [&NiagaraDataInterfaceCurve, &NiagaraDataInterfaceColorCurve, &NiagaraDataInterfaceCamera]
 usedDynamicInputs [
  &MakeFloatFromVector2d
  &MaskFloatBySpawnGroup
  &MakeFloatFromVector
 ]
 usedModules [
  &self:SystemSpawnScript.SystemScriptSource.SystemScriptGraph.NiagaraNodeAssignment_0.SetVariables_#1
  &SystemState
 ]
}
emitters {
 Blur {
  emitterUpdateScript [
   &EmitterState()
   &SpawnBurst_Instantaneous(
    Spawn Count=&MakeIntFromFloat(Float=&Multiply_Float(A=20, B=$System.RadiusScale))
    Spawn Time=0.1
    Use Spawn Probability=false
    UseLoopCountLimit=true
   )
  ]
  particleSpawnScript [
   &InitializeParticle(
    Color Mode=::NewEnumerator0
    Lifetime=&RandomRangeFloat(Minimum=0.5, Maximum=0.84)
    Sprite Size=&Multiply_Vector2D_ByFloat(Float=$System.RadiusScale, Vector2D=<100 200>)
    Sprite Size Mode=::NewEnumerator4
    UsePositionOffset=false
    Write Material Random=true
   )
   &ShapeLocation(
    Disc Coverage=0.2381
    MaskSpawnGroup=false
    Override Seed=false
    Ring Radius=400
    Rotation Coordinate Space=::World
    Shape Primitive=::NewEnumerator4
    UseOwnerScale=true
   ) {module=ShapeLocation001}
   &AddVelocity(
    Module.Use Falloff Distance=false
    Override Seed=false
    Velocity Mode=::NewEnumerator1
    Velocity Speed=&RandomRangeFloat(Minimum=1431.3961, Maximum=1688.7693)
   )
   &AddVelocity(Rotation Coordinate Space=::World, Velocity=<0 0 206.6583>) {module=AddVelocity001}
   &ApplyInitialForces()
  ]
  particleUpdateScript [
   &ParticleState()
   &Drag(Drag=4, Use Linear Drag=true, Use Rotational Drag=true)
   &SolveForcesAndVelocity(Clamp Velocity=false, Limit Acceleration=false)
   &ScaleColor(
    Scale Alpha=&FloatFromCurve(
     FloatCurve=DataInterface() {propertyValues {Curve {
        defaultValue=1
        keys [
         {arriveTangent=-2.5249, interpMode=RCIM_Cubic, leaveTangent=-2.5249, tangentMode=RCTM_User, time=0.1, value=1.1}
         {arriveTangent=-0.0405, interpMode=RCIM_Cubic, leaveTangent=-0.0405, tangentMode=RCTM_User, time=1, value=0}
        ]
       }}}
     Scale Curve=0.2
    )
    ScaleA=true
    ScaleRGB=true
   )
  ]
  rendererClasses [&NiagaraSpriteRendererProperties]
  renderers [{rendererClass=&NiagaraSpriteRendererProperties}]
 }
 FN_Debris_Mesh {
  emitterUpdateScript [
   &EmitterScalability(
    Module.bApplySpawnCountScale=false
    Module.bApplySpawnCountScaleByDistanceFraction=false
    Module.bUseMaxDistance=false
    Module.bUseMinDistance=false
   ) {enabled=false}
   &EmitterLifeCycle() {enabled=false}
   &EmitterState()
   &SpawnBurst_Instantaneous(
    Spawn Count=&MakeIntFromFloat(Float=&Multiply_Float(A=40, B=&Multiply_Float(A=$User.DebrisSpawnRate, B=$System.RadiusScale)))
    Spawn Time=0.1
    Use Spawn Probability=false
   )
  ]
  particleSpawnScript [
   /Niagara/Modules/Spawn/Initialization/InitializeParticle(
    Lifetime=&RandomRangeFloat(Minimum=1, Maximum=2)
    Mesh Scale=&Multiply_VectorByFloat(
     Vector=&RandomRangeVector(Minimum=<0.1 0.1 0.1>, Maximum=<1 1 1>)
     Float=&FloatFromCurve(
      FloatCurve=DataInterface() {propertyValues {Curve {keys [{arriveTangent=-1.5969, interpMode=RCIM_Cubic, leaveTangent=-1.5969, tangentMode=RCTM_User, time=0, value=1}, <at 1 0.25 RCIM_Cubic>]}}}
      CurveIndex=&RandomRangeFloat(Minimum=0, Maximum=1)
      Scale Curve=3
     )
    )
    Module.Write Color=true
    Module.Write Lifetime=true
    Module.Write Mass=true
    Module.Write Material Random=true
    Module.Write Position=false
    Module.Write Scale=true
    Module.Write SpriteRotation=false
    Module.Write SpriteSize=false
    Position=&SimulationPosition()
   )
   &ShapeLocation(
    Disc Coverage=0.2381
    MaskSpawnGroup=false
    Override Seed=false
    Ring Radius=400
    Rotation Coordinate Space=::World
    Shape Primitive=::NewEnumerator4
    UseOwnerScale=true
   )
   /Niagara/Modules/Spawn/Orientation/V2/InitialMeshOrientation(
    Orientation Vector=/Niagara/DynamicInputs/Random/RandomVector(
     Vector Scale:NiagaraFloat=1
     Constraint Vector:Vector3f=<0 0 1>
     Randomness Mode:ENiagaraRandomnessMode=::NewEnumerator0
     Random Seed:NiagaraInt32=0
     Fixed Random Seed:NiagaraBool=false
    )
    Use Orientation Vector=true
    Use Rotation Vector=false
   )
   &MeshRotationForce(
    "Lever Radius (cm)":NiagaraFloat=&UniformRangedFloat(Minimum=200, Maximum=400)
    Rotation=&RandomVector(
     Vector Scale:NiagaraFloat=1
     Constraint Vector:Vector3f=<0 0 1>
     Randomness Mode:ENiagaraRandomnessMode=::NewEnumerator0
     Random Seed:NiagaraInt32=0
     Fixed Random Seed:NiagaraBool=false
    )
   )
   &AddVelocity(
    Module.Use Falloff Distance=false
    Override Seed=false
    Velocity Mode=::NewEnumerator1
    Velocity Speed=&RandomRangeFloat(Minimum=0, Maximum=1138.9447)
   )
   &AddVelocity(Rotation Coordinate Space=::World, Velocity=&RandomRangeVector(Minimum=<0 0 1000>, Maximum=<0 0 3000>)) {module=AddVelocity001}
   &AddVelocityInCone(
    Cone Angle=135
    Module.Use Velocity Falloff On Cone Axis=false
    Velocity Strength=&UniformRangedFloat(Minimum=300, Maximum=700)
   ) {enabled=false}
   &ApplyInitialForces()
   /SetVariables(Particles.RandomGravity:NiagaraFloat=&UniformRangedFloat(Minimum=-755.642, Maximum=-982.9998)) {isSetParameters=true, module=SetVariables_#2}
  ]
  particleUpdateScript [
   &ParticleState()
   &ScaleMeshSize(Scale Factor=&VectorFromFloat(Value=&FloatFromCurve(FloatCurve=DataInterface() {propertyValues {Curve {keys [<at 0.6 1>,<at 1 0>]}}})))
   &AccelerationForce(Acceleration=&MakeVector(X=0, Y=0, Z=$Particles.RandomGravity))
   &Drag(
    Drag=&FloatFromCurve(
     FloatCurve=DataInterface() {propertyValues {Curve {keys [
         <at 0.2 1 RCIM_Cubic>
         {arriveTangent=-0.0326, interpMode=RCIM_Cubic, leaveTangent=-0.0326, tangentMode=RCTM_User, time=0.3, value=0}
         <at 1 0 RCIM_Cubic>
        ]}}}
     Scale Curve=5
    )
    Use Linear Drag=true
    Use Rotational Drag=true
   )
   &SolveForcesAndVelocity(Clamp Velocity=false, Limit Acceleration=false)
   &SolveRotationalForcesAndVelocity()
   &Color()
  ]
  rendererClasses [&NiagaraMeshRendererProperties]
  renderers [{rendererClass=&NiagaraMeshRendererProperties, rendererIndex=0}]
 }
 FN_Debris_Mesh_Large {
  emitterUpdateScript [
   &EmitterScalability(
    Module.bApplySpawnCountScale=false
    Module.bApplySpawnCountScaleByDistanceFraction=false
    Module.bUseMaxDistance=false
    Module.bUseMinDistance=false
   ) {enabled=false}
   &EmitterLifeCycle() {enabled=false}
   &EmitterState()
   &SpawnBurst_Instantaneous(
    Spawn Count=&MakeIntFromFloat(Float=&Multiply_Float(A=3, B=&Multiply_Float(A=$User.DebrisSpawnRate, B=$System.RadiusScale)))
    Spawn Time=0.1
    Use Spawn Probability=false
   )
  ]
  particleSpawnScript [
   &InitializeParticle(
    Lifetime=&RandomRangeFloat(Minimum=1, Maximum=2)
    Mesh Scale=&Multiply_VectorByFloat(
     Vector=&RandomRangeVector(Minimum=<0.1 0.1 0.1>, Maximum=<1 1 1>)
     Float=&FloatFromCurve(
      FloatCurve=DataInterface() {propertyValues {Curve {keys [{arriveTangent=-1.5969, interpMode=RCIM_Cubic, leaveTangent=-1.5969, tangentMode=RCTM_User, time=0, value=1}, <at 1 0.55 RCIM_Cubic>]}}}
      CurveIndex=&RandomRangeFloat(Minimum=0, Maximum=1)
      Scale Curve=10
     )
    )
    Module.Write Color=true
    Module.Write Lifetime=true
    Module.Write Mass=true
    Module.Write Material Random=true
    Module.Write Position=false
    Module.Write Scale=true
    Module.Write SpriteRotation=false
    Module.Write SpriteSize=false
    Position=&SimulationPosition()
   )
   &ShapeLocation(
    Disc Coverage=0.2381
    MaskSpawnGroup=false
    Override Seed=false
    Ring Radius=400
    Rotation Coordinate Space=::World
    Shape Primitive=::NewEnumerator4
    UseOwnerScale=true
   )
   &InitialMeshOrientation(
    Orientation Vector=&RandomVector(
     Vector Scale:NiagaraFloat=1
     Constraint Vector:Vector3f=<0 0 1>
     Randomness Mode:ENiagaraRandomnessMode=::NewEnumerator0
     Random Seed:NiagaraInt32=0
     Fixed Random Seed:NiagaraBool=false
    )
    Use Orientation Vector=true
    Use Rotation Vector=false
   )
   &MeshRotationForce(
    "Lever Radius (cm)":NiagaraFloat=&UniformRangedFloat(Minimum=200, Maximum=400)
    Rotation=&RandomVector(
     Vector Scale:NiagaraFloat=0.5
     Constraint Vector:Vector3f=<0 0 1>
     Randomness Mode:ENiagaraRandomnessMode=::NewEnumerator0
     Random Seed:NiagaraInt32=0
     Fixed Random Seed:NiagaraBool=false
    )
   )
   &AddVelocity(
    Module.Use Falloff Distance=false
    Override Seed=false
    Velocity Mode=::NewEnumerator1
    Velocity Speed=&RandomRangeFloat(Minimum=0, Maximum=1138.9447)
   )
   &AddVelocity(
    Rotation Coordinate Space=::World
    Velocity=&RandomRangeVector(Minimum=<0 0 715.0895>, Maximum=<0 0 1525.486>)
   ) {module=AddVelocity001}
   &AddVelocityInCone(
    Cone Angle=135
    Module.Use Velocity Falloff On Cone Axis=false
    Velocity Strength=&UniformRangedFloat(Minimum=300, Maximum=700)
   ) {enabled=false}
   &ApplyInitialForces()
   /SetVariables(Particles.RandomGravity:NiagaraFloat=&UniformRangedFloat(Minimum=-755.642, Maximum=-982.9998)) {isSetParameters=true, module=SetVariables_#2}
  ]
  particleUpdateScript [
   &ParticleState()
   &ScaleMeshSize(Scale Factor=&VectorFromFloat(Value=&FloatFromCurve(FloatCurve=DataInterface() {propertyValues {Curve {keys [<at 0.6 1>,<at 1 0>]}}})))
   &AccelerationForce(Acceleration=&MakeVector(X=0, Y=0, Z=$Particles.RandomGravity))
   &Drag(
    Drag=&FloatFromCurve(
     FloatCurve=DataInterface() {propertyValues {Curve {keys [
         <at 0.2 1 RCIM_Cubic>
         {arriveTangent=-0.0326, interpMode=RCIM_Cubic, leaveTangent=-0.0326, tangentMode=RCTM_User, time=0.3, value=0}
         <at 1 0 RCIM_Cubic>
        ]}}}
     Scale Curve=5
    )
    Use Linear Drag=true
    Use Rotational Drag=true
   )
   &Collision(
    Age Colliding Particles=false
    Collision Enabled=&SetBoolByFloatComparison(A=$Particles.Age, B=0.3)
    KillOnCollision=false
    Module.Control Rest State Via Surface Normal=false
    Module.EnableMaxCollisionCount=false
    Module.Randomize Collision Normal Vector=false
   )
   &SolveForcesAndVelocity(Clamp Velocity=false, Limit Acceleration=false)
   &SolveRotationalForcesAndVelocity()
   &Color()
  ]
  rendererClasses [&NiagaraMeshRendererProperties]
  renderers [{rendererClass=&NiagaraMeshRendererProperties, rendererIndex=0}]
 }
 FakeVolume {
  emitterUpdateScript [
   &EmitterState(Life Cycle Mode=::NewEnumerator1, Loop Behavior=::NewEnumerator1, Loop Delay=$System.BurstTime, Loop Duration=0.15, UseLoopDelay=true) {module=EmitterState001}
   &SpawnBurst_Instantaneous(Spawn Count=8, Use Spawn Probability=false, UseLoopCountLimit=false)
  ]
  particleSpawnScript [
   /Niagara/Modules/Spawn/Initialization/V2/InitializeParticle(
    Color Mode=::NewEnumerator0
    Lifetime=0.35
    Sprite Size Mode=::NewEnumerator3
    Uniform Sprite Size=&Multiply_Float(A=700, B=$System.RadiusScale)
    UsePositionOffset=false
    Write Material Random=true
   )
   &ShapeLocation(
    Disc Coverage=0.2381
    MaskSpawnGroup=false
    Override Seed=false
    Ring Radius=400
    Rotation Coordinate Space=::World
    Shape Primitive=::NewEnumerator4
    UseOwnerScale=true
   ) {module=ShapeLocation001}
  ]
  particleUpdateScript [
   &ParticleState()
   &ScaleSpriteSize(
    Uniform Curve Scale=&RandomRangeFloat(Minimum=0.8, Maximum=1, Evaluation Type=::NewEnumerator1)
    Uniform Curve Sprite Scale=DataInterface() {propertyValues {Curve {
       defaultValue=0
       keys [
        {arriveTangent=4.1785, interpMode=RCIM_Cubic, leaveTangent=4.1785, tangentMode=RCTM_User, time=0, value=0.15}
        {arriveTangent=1.1179, interpMode=RCIM_Cubic, leaveTangent=1.1179, tangentMode=RCTM_User, time=0.2, value=0.65}
        {arriveTangent=0.0224, interpMode=RCIM_Cubic, leaveTangent=0.0224, tangentMode=RCTM_User, time=1, value=1}
       ]
      }}}
   )
   &DynamicMaterialParameters(
    Index 0 Param 3=1
    Index 0 Param 4=500
    Module.Index0Float0Write=true
    Module.Index0Float1Write=true
    Module.Index0Float2Write=true
    Module.Index0Float3Write=true
   )
   &ScaleColor(
    Scale Alpha=&FloatFromCurve(
     FloatCurve=DataInterface() {propertyValues {Curve {
        defaultValue=1
        keys [
         {interpMode=RCIM_Cubic, tangentMode=RCTM_User, time=0, value=0}
         {arriveTangent=-3.5856, interpMode=RCIM_Cubic, leaveTangent=-3.5856, tangentMode=RCTM_User, time=0.0524, value=1}
         {arriveTangent=-1.0553, interpMode=RCIM_Cubic, leaveTangent=-1.0553, time=0.2, value=0.3}
         {tangentMode=RCTM_User, time=1, value=0}
        ]
       }}}
     Scale Curve=&RandomRangeFloat(Minimum=0.1, Maximum=1, Evaluation Type=::NewEnumerator1)
    )
    ScaleA=true
    ScaleRGB=true
   )
   &JitterPosition(Jitter Amount=300) {enabled=false}
  ]
  rendererClasses [&NiagaraSpriteRendererProperties]
  renderers [{rendererClass=&NiagaraSpriteRendererProperties}]
 }
 Smoke {
  emitterUpdateScript [
   &EmitterState()
   &SpawnBurst_Instantaneous(Spawn Count=24, Use Spawn Probability=false, UseLoopCountLimit=true)
  ]
  particleSpawnScript [
   &InitializeParticle(
    Color Maximum=<rgba 0.3854 0.3854 0.3854 0.4957>
    Color Minimum=<rgba 0.2552 0.2552 0.2552 1.0412>
    Color Mode=::NewEnumerator2
    Lifetime=&FloatFromCurve(
     FloatCurve=DataInterface() {propertyValues {Curve {keys [{arriveTangent=-1.9457, interpMode=RCIM_Cubic, leaveTangent=-1.9457, tangentMode=RCTM_User, time=0, value=1}, <at 1 0.25 RCIM_Cubic>]}}}
     CurveIndex=&RandomRangeFloat(Minimum=0, Maximum=1)
     Scale Curve=1.25
    )
    Sprite Rotation Mode=::NewEnumerator2
    Sprite Size=&Multiply_Vector2D_ByFloat(Float=$System.RadiusScale, Vector2D=&RandomRangeVector2D(Minimum=<80 80>, Maximum=<200 200>))
    Sprite Size Mode=::NewEnumerator4
    Sprite UV Mode=::NewEnumerator3
    UsePositionOffset=false
    Write Material Random=true
   )
   &CameraOffset(Camera Offset Amount=50)
   &ShapeLocation(
    Disc Coverage=0.0714
    MaskSpawnGroup=false
    Override Seed=false
    Ring Radius=350
    Rotation Coordinate Space=::World
    Shape Primitive=::NewEnumerator4
    UseOwnerScale=true
   ) {module=ShapeLocation001}
   &AddVelocity(
    Module.Use Falloff Distance=false
    Origin Offset Coordinate Space=::World
    Override Seed=false
    Velocity Mode=::NewEnumerator1
    Velocity Speed=&RandomRangeFloat(Minimum=1495.1179, Maximum=1519.1489)
   )
  ]
  particleUpdateScript [
   &Drag(Drag=&RandomRangeFloat(Minimum=6, Maximum=11), Use Linear Drag=true, Use Rotational Drag=true)
   &AccelerationForce(Acceleration=&RandomRangeVector(Minimum=<0 0 -100>, Maximum=<0 0 200>))
   &SolveForcesAndVelocity(Clamp Velocity=false, Limit Acceleration=false)
   &ParticleState()
   &ScaleSpriteSize(Uniform Curve Sprite Scale=DataInterface() {propertyValues {Curve {keys [
        {arriveTangent=4.1286, interpMode=RCIM_Cubic, leaveTangent=4.1286, tangentMode=RCTM_User, time=0, value=0.475}
        {arriveTangent=1.2171, interpMode=RCIM_Cubic, leaveTangent=1.2171, tangentMode=RCTM_User, time=0.3, value=1.125}
        <at 1 1.4 RCIM_Cubic>
       ]}}})
   /SetVariables(Particles.SubImageIndex=&FloatFromCurve(
     FloatCurve=DataInterface() {propertyValues {Curve {keys [
         {arriveTangent=2.5597, interpMode=RCIM_Cubic, leaveTangent=2.5597, tangentMode=RCTM_User, time=0, value=0}
         {arriveTangent=1, interpMode=RCIM_Cubic, leaveTangent=1, time=0.1, value=0.75}
         {interpMode=RCIM_Cubic, leaveTangent=1.1425, time=1, value=1}
        ]}}}
     Scale Curve=63
    )) {isSetParameters=true}
   &DynamicMaterialParameters(
    Index 0 Param 1=&FloatFromCurve(FloatCurve=DataInterface() {propertyValues {Curve {keys [{arriveTangent=3.9017, interpMode=RCIM_Cubic, leaveTangent=3.9017, tangentMode=RCTM_User, time=0, value=0.35}, <at 0.5 1 RCIM_Cubic>]}}})
    Index 0 Param 4=1
    Module.Index0Float0Write=true
    Module.Index0Float1Write=true
    Module.Index0Float2Write=true
    Module.Index0Float3Write=true
   )
   &ScaleColor(
    Scale Alpha=&FloatFromCurve(FloatCurve=DataInterface() {propertyValues {Curve {keys [{arriveTangent=-2.6923, interpMode=RCIM_Cubic, leaveTangent=-2.6923, tangentMode=RCTM_User, time=0.1, value=1}, <at 1 0 RCIM_Cubic>]}}})
    Scale RGB=&MakeVectorFromLinearColorRGB(Color=&ColorFromCurve(ColorCurve=DataInterface() {propertyValues {
        AlphaCurve {keys []}
        BlueCurve {keys [{arriveTangent=-9.5363, interpMode=RCIM_Cubic, leaveTangent=-9.5363, tangentMode=RCTM_User, time=0, value=4}, <at 1 1 RCIM_Cubic>]}
        GreenCurve {keys [{arriveTangent=-9.5363, interpMode=RCIM_Cubic, leaveTangent=-9.5363, tangentMode=RCTM_User, time=0, value=4}, <at 1 1 RCIM_Cubic>]}
        RedCurve {keys [{arriveTangent=-9.5363, interpMode=RCIM_Cubic, leaveTangent=-9.5363, tangentMode=RCTM_User, time=0, value=4}, <at 1 1 RCIM_Cubic>]}
       }}))
    ScaleA=true
    ScaleRGB=true
   )
  ]
  rendererClasses [&NiagaraSpriteRendererProperties]
  renderers [{rendererClass=&NiagaraSpriteRendererProperties}]
 }
 Sparks {
  simTarget=GPUComputeSim
  emitterUpdateScript [
   &EmitterState()
   &SpawnBurst_Instantaneous(
    Spawn Count=&MakeIntFromFloat(Float=&Multiply_Float(A=3000, B=$System.RadiusScale))
    Spawn Time=0.05
    Use Spawn Probability=false
    UseLoopCountLimit=false
   )
  ]
  particleSpawnScript [
   &InitializeParticle(
    Color Maximum=<rgba 0.3385 0.3385 0.3385 1>
    Color Minimum=<rgba 0.0677 0.0677 0.0677 1>
    Color Mode=::NewEnumerator2
    Lifetime=&RandomRangeFloat(Minimum=1, Maximum=2)
    Sprite Size Mode=::NewEnumerator3
    Uniform Sprite Size=&FloatFromCurve(
     FloatCurve=DataInterface() {propertyValues {Curve {keys [{arriveTangent=-2.2424, interpMode=RCIM_Cubic, leaveTangent=-2.2424, tangentMode=RCTM_User, time=0, value=1.5}, <at 1 0.85 RCIM_Cubic>]}}}
     CurveIndex=&RandomRangeFloat(Minimum=0, Maximum=1)
     Scale Curve=2
    )
    UsePositionOffset=false
    Write Material Random=true
   )
   &ShapeLocation(
    Disc Coverage=0.2381
    MaskSpawnGroup=false
    Override Seed=false
    Ring Radius=376.7057
    Rotation Coordinate Space=::World
    Shape Primitive=::NewEnumerator4
    UseOwnerScale=true
   ) {module=ShapeLocation001}
   &AddVelocity(
    Module.Use Falloff Distance=false
    Override Seed=false
    Velocity Mode=::NewEnumerator1
    Velocity Speed=&FloatFromCurve(
     FloatCurve=DataInterface() {propertyValues {Curve {keys [<at 0 1>,<at 0.4 1>,<at 0.5 0.3>,<at 0.9 0.3>,<at 1 0>]}}}
     CurveIndex=&RandomRangeFloat(Minimum=0, Maximum=1)
     Scale Curve=2000
    )
   )
   &AddVelocity(Velocity=<500 0 0>) {module=AddVelocity001}
  ]
  particleUpdateScript [
   &ParticleState()
   &CurlNoiseForce(
    Bias Noise=false
    Noise Frequency=10
    Noise Strength=&FloatFromCurve(
     FloatCurve=DataInterface() {propertyValues {Curve {
        defaultValue=1
        keys [
         {arriveTangent=-3.2557, interpMode=RCIM_Cubic, leaveTangent=-3.2557, tangentMode=RCTM_User, time=0, value=1}
         {arriveTangent=-0.7893, interpMode=RCIM_Cubic, leaveTangent=-0.7893, tangentMode=RCTM_User, time=0.1, value=0.25}
         {interpMode=RCIM_Cubic, leaveTangent=-8.0089, time=1, value=0}
        ]
       }}}
     Scale Curve=10061.9443
    )
    Pan Noise=false
   )
   &Drag(Drag=5, Use Linear Drag=true, Use Rotational Drag=true)
   &PointForce(Force Strength=-259.227, Module.Use Falloff=false, Module.UseLinearForce=false)
   &SolveForcesAndVelocity(Clamp Velocity=false, Limit Acceleration=false)
   &ScaleSpriteSize(
    Scale Sprite Size Mode=::NewEnumerator0
    Uniform Scale Factor=/Niagara/DynamicInputs/Camera/CameraFade(
     Camera Query:NiagaraDataInterfaceCamera=DataInterface()
     Fade Offset:NiagaraFloat=100
     Fade Distance:NiagaraFloat=500
     Use Owner Position:NiagaraBool=true
    )
   )
   &ScaleSpriteSizeBySpeed(Initial Sprite Size:Vector2f=$Particles.SpriteSize, Max Scale Factor=<1 50>, Velocity Threshold=2000)
   &ScaleColor(
    Scale Alpha=&FloatFromCurve(
     FloatCurve=DataInterface() {propertyValues {Curve {keys [{arriveTangent=-3.8916, interpMode=RCIM_Cubic, leaveTangent=-3.8916, tangentMode=RCTM_User, time=0.2, value=1}, <at 1 0 RCIM_Cubic>]}}}
     Scale Curve=0.5
    )
    ScaleA=true
    ScaleRGB=true
   )
  ]
  rendererClasses [&NiagaraSpriteRendererProperties]
  renderers [{rendererClass=&NiagaraSpriteRendererProperties}]
 }
}
summary {
 systemName=NS_CircleWave
 emitters [
  {emitterName=Smoke, rendererClasses [&NiagaraSpriteRendererProperties]}
  {emitterName=Sparks, simTarget=GPUComputeSim, rendererClasses [&NiagaraSpriteRendererProperties]}
  {emitterName=FN_Debris_Mesh, rendererClasses [&NiagaraMeshRendererProperties]}
  {emitterName=Blur, rendererClasses [&NiagaraSpriteRendererProperties]}
  {emitterName=FakeVolume, rendererClasses [&NiagaraSpriteRendererProperties]}
  {emitterName=FN_Debris_Mesh_Large, rendererClasses [&NiagaraMeshRendererProperties]}
 ]
 userVariables [
  User.Color {type=LinearColor, description="", defaultValue {struct=LinearColor, value=<rgba 1.852 0 20 1>}}
  User.DebrisSpawnRate {type=NiagaraFloat, defaultValue=1, description=""}
  User.RadiusScale {type=NiagaraFloat, defaultValue=2, description=""}
  User.Size {type=NiagaraFloat, defaultValue=50, description=""}
  User.Size_Scale {type=NiagaraFloat, defaultValue=600, description=""}
  User.Timing_BurstDelay {type=NiagaraFloat, defaultValue=0, description=""}
 ]
}
